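; Construction script for the March statue prop in the Creative March set, which tints the statue through two material color parameters.
; It copies the RandomColor_A and RandomColor_B variables into the material vector parameters of the same names.
; Both writes go through editor-only instance parameter setters, and the script itself does no randomization.
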(fn ConstructionScript
  (EditorOnlyParameters|EditorOnlySetInstanceMaterialVectorParameter "RandomColor_A" (Variables|Colors|GetRandomColor_A))
  (EditorOnlyParameters|EditorOnlySetInstanceMaterialVectorParameter "RandomColor_B" (Variables|Colors|GetRandomColor_B)))
