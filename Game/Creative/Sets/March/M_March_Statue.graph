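# M_March_Statue is a Default Lit surface material, named for a statue in the Creative March set.
# Its graph holds a single Texture Sample of T_March_Statue, with no parameters or math.
# The graph shows no connection from that sample to any material output.
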Material /Game/Creative/Sets/March/M_March_Statue
graph {
 properties {pixelDepthOffsetMode=PDOM_Legacy, usedShadingModels=Default Lit}
 nodes [TextureSample() {texture=&/T_March_Statue}]
}
